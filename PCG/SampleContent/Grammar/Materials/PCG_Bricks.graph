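# Procedural brick material for the PCG Grammar sample content, usable on instanced static meshes, that draws a brick pattern without textures.
# UVs are scaled by tile counts, with alternate rows shifted by an offset to form a running bond.
# Per-brick cell indices seed vector noise that slightly varies the brick color.
# A BoxMask-2D on each cell's fractional UVs, thresholded by Step, blends between brick color and LimitColor for the joints.
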Material /PCG/SampleContent/Grammar/Materials/PCG_Bricks
graph {
 properties {bUsedWithInstancedStaticMeshes=true, usedShadingModels=Default Lit}
 nodes [
  n1 = MaterialFunctionCall(Float2=TextureCoordinate()) {materialFunction=&BreakOutFloat2Components}
  n2 = Multiply(A=@n1.G, B=ScalarParameter() {defaultValue=20, parameterName=NumTiles_Y})
  n3 = Ceil(None=@n2)
  n4 = Constant() {r=2}
  n5 = Add(
   A=Multiply(A=ScalarParameter() {defaultValue=10, parameterName=NumTiles_X}, B=@n1.R)
   B=Multiply(A=Fmod(A=@n3, B=@n4), B=ScalarParameter() {defaultValue=0.5, parameterName=Offset})
  )
  n6 = Ceil(None=@n5)
  n7 = VectorNoise(World Position=AppendVector(A=AppendVector(A=@n6, B=@n3), B=Constant()))
  n8 = AppendVector(A=Frac(None=@n5), B=Frac(None=@n2))
  n9 = Constant() {r=0.5}
  n10 = VectorParameter() {parameterName=BrickSizeAndFalloff, defaultValue=<rgba 0.8 0.8 0.1 0>}
  LinearInterpolate(
   A=LinearInterpolate(
    A=VectorParameter() {parameterName=BrickColor, defaultValue=<rgba 0.0833 0.0311 0.0148 1>}.RGB
    B=ComponentMask(None=@n7) {r=true}
    Alpha=ScalarParameter() {defaultValue=0.05, parameterName=LuminanceVariation}
   )
   B=VectorParameter() {parameterName=LimitColor}.RGB
   Alpha=Step(
    Y=MaterialFunctionCall!(A=@n8, B=@n9, Bounds=@n10.RGB, Edge Falloff=@n10.RGB) {materialFunction=/Engine/Functions/Engine_MaterialFunctions01/Gradient/BoxMask-2D}.Result
    X=ScalarParameter() {defaultValue=0.3, parameterName=MaskSize}
   ) {constX=0.3}
  )
 ]
}
