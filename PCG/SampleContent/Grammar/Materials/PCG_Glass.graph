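# PCG_Glass is a two-sided translucent glass material in the PCG Grammar sample content, enabled for instanced static meshes.
# It uses Default Lit shading with surface translucency lighting and index-of-refraction refraction.
# A Fresnel term blends from the index of air to that of glass, alongside a white color constant and a few fixed scalars.
# It has no textures or parameters.
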Material /PCG/SampleContent/Grammar/Materials/PCG_Glass
graph {
 properties {
  bUsedWithInstancedStaticMeshes=true
  blendMode=BLEND_Translucent
  refractionMethod=RM_IndexOfRefraction
  translucencyLightingMode=TLM_Surface
  twoSided=true
  usedShadingModels=Default Lit
 }
 nodes [
  Constant3Vector() {constant=<rgba 1 1 1 0>}
  Constant() {r=1}
  Constant()
  Constant() {r=0.05}
  LinearInterpolate(A=Constant() {r=1}, B=Constant() {r=1.52}, Alpha=Fresnel())
 ]
}
